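# One-shot mono MetaSound source for The Baron boss's Phase 2 voice lines in VividRazor.
# It wraps the shared MSS_VO_BR_Boss graph, feeding it the wave array plus delay, probability, pitch and volume inputs.
# Only the two-line Wave Array is wired; the second, six-wave array and second delay pair are unconnected.
# Playback uses shared NPC alert attenuation, boss chatter concurrency limited to one per owner, and a music sidechain bus send.
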MetaSoundSource /VividRazor_Baron/Audio/VO/Abilities/MSP_VO_VividRazor_TheBaron_Phase2
inputs [
 "Delay Max (Seconds)" {type=Time}
 "Delay Max (Seconds) 1" {type=Time}
 "Delay Min (Seconds)" {type=Time}
 "Delay Min (Seconds) 1" {type=Time}
 "Probability (0-1)" {type=Float}
 "Random Pitch Max (Semitones)" {type=Float}
 "Random Pitch Min (Semitones)" {type=Float}
 "Random Volume Change Max (Decibels)" {type=Float}
 "Random Volume Change Min (Decibels)" {type=Float}
 UE.Source.OnPlay {type=Trigger}
 Wave Array {type=WaveAsset:Array}
 Wave Array 1 {type=WaveAsset:Array}
]
interfaces [UE.OutputFormat.Mono:1.0, UE.Source.OneShot:1.0, UE.Source:1.1]
outputs [
 "UE.OutputFormat.Mono.Audio:0" {type=Audio}
 "UE.OutputFormat.Mono.Audio:0 1" {type=Audio}
 UE.Source.OneShot.OnFinished {type=Trigger}
]
pages [Default {
  input_defaults {
   "Delay Max (Seconds)"=0
   "Delay Max (Seconds) 1"=2.5
   "Delay Min (Seconds)"=0
   "Delay Min (Seconds) 1"=2.5
   "Probability (0-1)"=1
   "Random Pitch Max (Semitones)"=0
   "Random Pitch Min (Semitones)"=0
   "Random Volume Change Max (Decibels)"=0
   "Random Volume Change Min (Decibels)"=0
   UE.Source.OnPlay=null
   Wave Array [/VividRazor_Baron/Audio/VO/Waves/VO_VividRazor_TheBaron_Phase2_010, /VividRazor_Baron/Audio/VO/Waves/VO_VividRazor_TheBaron_Phase2_011]
   Wave Array 1 [
    &VO_VividRazor_TheBaron_Phase2_010
    &VO_VividRazor_TheBaron_Phase2_011
    /VividRazor_Baron/Audio/VO/Waves/VO_VividRazor_TheBaron_Phase2_020
    /VividRazor_Baron/Audio/VO/Waves/VO_VividRazor_TheBaron_Phase2_030
    /VividRazor_Baron/Audio/VO/Waves/VO_VividRazor_TheBaron_Phase2_040
    /VividRazor_Baron/Audio/VO/Waves/VO_VividRazor_TheBaron_Phase2_050
   ]
  }
  nodes [
   InitVariable.Audio()
   MSS_VO_BR_Boss_0 = /Game/Sounds/Fort_Human_VO_BR/NPCs/Shared/MSS_VO_BR_Boss(
    "Delay Max (Seconds)"=^"Delay Max (Seconds)"
    "Delay Min (Seconds)"=^"Delay Min (Seconds)"
    "Probability (0-1)"=^"Probability (0-1)"
    "Random Pitch Max (Semitones)"=^"Random Pitch Max (Semitones)"
    "Random Pitch Min (Semitones)"=^"Random Pitch Min (Semitones)"
    "Random Volume Change Max (Decibels)"=^"Random Volume Change Max (Decibels)"
    "Random Volume Change Min (Decibels)"=^"Random Volume Change Min (Decibels)"
    UE.Source.OnPlay=^UE.Source.OnPlay
    Wave Array=^Wave Array
   )
  ]
  links [
   {dst="@UE.OutputFormat.Mono.Audio:0", src=MSS_VO_BR_Boss_0.UE.OutputFormat.Mono.Audio:0}
   {dst="@UE.Source.OneShot.OnFinished", src=MSS_VO_BR_Boss_0.UE.Source.OneShot.OnFinished}
  ]
 }]
properties {
 assetImportData=&self:AssetImportData
 attenuationSettings=/Game/Sounds/Fort_Human_VO_BR/NPCs/Shared/NPC_VO_Alerted_Att
 maxDistance=12500
 sampleRate=48000
 soundClassObject=&Fort_VO_Enemy_Critical
 volume=3
 busSends [{
   audioBus=&AB_Sidechain_FortMusic_Vol
   maxSendDistance=13000
   maxSendLevel=0.3
   minSendDistance=6000
   minSendLevel=0.3
   sendLevel=0.3
  }]
 concurrencyOverrides {bLimitToOwner=true, maxCount=1, resolutionRule=PreventNew}
 concurrencySet [&NPC_Boss_Combat_Chatter_Conc]
}
